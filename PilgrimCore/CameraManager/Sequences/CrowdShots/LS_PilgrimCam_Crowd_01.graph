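# Level Sequence for a crowd camera shot in the Pilgrim camera manager, playing one 600-frame cine camera move at 60 fps.
# A Cine Camera Actor with a spawn track travels sideways along Y with a slight descent while its yaw pans, at constant X.
# Its camera component holds focal length, manual focus distance and a very wide aperture constant.
# A single camera cut section spans the whole move.
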LevelSequence /PilgrimCore/CameraManager/Sequences/CrowdShots/LS_PilgrimCam_Crowd_01
bindings [
 CameraComponent {
  id=#1
  parent_id=#2
  possessed_class=CineCameraComponent
  tracks [
   {
    type=MovieSceneFloatTrack
    display_name=Current Focal Length
    sections [{type=MovieSceneFloatSection, channels [None {type=MovieSceneScriptingFloatChannel, default=36.3704}]}]
   }
   {
    type=MovieSceneFloatTrack
    display_name="Manual Focus Distance (Focus Settings)"
    sections [{type=MovieSceneFloatSection, channels [None {type=MovieSceneScriptingFloatChannel, default=3512.9009, const [3512.9009,1]}]}]
   }
   {
    type=MovieSceneFloatTrack
    display_name=Current Aperture
    sections [{type=MovieSceneFloatSection, channels [None {type=MovieSceneScriptingFloatChannel, default=0.5}]}]
   }
  ]
 }
 Cine Camera Actor {
  id=#2
  possessed_class=CineCameraActor
  tracks [
   {
    type=MovieScene3DTransformTrack
    sections [{
      type=MovieScene3DTransformSection
      channels [
       Location.X {type=MovieSceneScriptingDoubleChannel, default=-6371.6631, const [-3493.9663,2]}
       Location.Y {type=MovieSceneScriptingDoubleChannel, default=-552.7704, keys [[0,-1792.4756],[600,944.2214]]}
       Location.Z {type=MovieSceneScriptingDoubleChannel, default=556, keys [[0,782.5451],[600,755.023]]}
       Rotation.X {type=MovieSceneScriptingDoubleChannel, default=0, const [0,2]}
       Rotation.Y {type=MovieSceneScriptingDoubleChannel, default=0, const [0,2]}
       Rotation.Z {type=MovieSceneScriptingDoubleChannel, default=190, keys [[0,150],[600,190]]}
       Scale.X {type=MovieSceneScriptingDoubleChannel, default=1, const [1,2]}
       Scale.Y {type=MovieSceneScriptingDoubleChannel, default=1, const [1,2]}
       Scale.Z {type=MovieSceneScriptingDoubleChannel, default=1, const [1,2]}
      ]
     }]
   }
   {type=MovieSceneSpawnTrack, sections [{type=MovieSceneSpawnSection, channels [None {type=MovieSceneScriptingBoolChannel, default=true}]}]}
  ]
 }
]
display_rate {denominator=1, numerator=60}
tick_resolution {denominator=1, numerator=24000}
tracks [{type=MovieSceneCameraCutTrack, sections [{type=MovieSceneCameraCutSection, end_frame=600, start_frame=0}]}]
