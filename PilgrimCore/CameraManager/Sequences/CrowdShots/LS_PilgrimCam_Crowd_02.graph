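# Level Sequence for PilgrimCam crowd shot 02, a ten-second camera shot in the Pilgrim camera manager's set of crowd shots.
# A Cine Camera Actor with a spawn track holds one constant transform, while its camera component sets fixed focal length, manual focus distance and aperture.
# A camera shake track on the component runs slightly past the camera cut section, the only layer that can add motion to the otherwise static framing.
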LevelSequence /PilgrimCore/CameraManager/Sequences/CrowdShots/LS_PilgrimCam_Crowd_02
bindings [
 CameraComponent {
  id=#1
  parent_id=#2
  possessed_class=CineCameraComponent
  tracks [
   {
    type=MovieSceneFloatTrack
    display_name=Current Focal Length
    sections [{type=MovieSceneFloatSection, channels [None {type=MovieSceneScriptingFloatChannel, default=24.8695}]}]
   }
   {
    type=MovieSceneFloatTrack
    display_name="Manual Focus Distance (Focus Settings)"
    sections [{type=MovieSceneFloatSection, channels [None {type=MovieSceneScriptingFloatChannel, default=7156.2363}]}]
   }
   {
    type=MovieSceneFloatTrack
    display_name=Current Aperture
    sections [{type=MovieSceneFloatSection, channels [None {type=MovieSceneScriptingFloatChannel, default=1}]}]
   }
   {type=MovieSceneCameraShakeTrack, sections [{type=MovieSceneCameraShakeSection, end_frame=634}]}
  ]
 }
 Cine Camera Actor {
  id=#2
  possessed_class=CineCameraActor
  tracks [
   {
    type=MovieScene3DTransformTrack
    sections [{
      type=MovieScene3DTransformSection
      channels [
       Location.X {type=MovieSceneScriptingDoubleChannel, default=-5094.4652, const [0.8543,1]}
       Location.Y {type=MovieSceneScriptingDoubleChannel, default=-295.3983, const [1248.8063,1]}
       Location.Z {type=MovieSceneScriptingDoubleChannel, default=440.448, const [648.4653,1]}
       Rotation.X {type=MovieSceneScriptingDoubleChannel, default=0, const [-0.2734,1]}
       Rotation.Y {type=MovieSceneScriptingDoubleChannel, default=-12.1388, const [8.8615,1]}
       Rotation.Z {type=MovieSceneScriptingDoubleChannel, default=-174.8193, const [206.354,1]}
       Scale.X {type=MovieSceneScriptingDoubleChannel, default=1, const [1,1]}
       Scale.Y {type=MovieSceneScriptingDoubleChannel, default=1, const [1,1]}
       Scale.Z {type=MovieSceneScriptingDoubleChannel, default=1, const [1,1]}
      ]
     }]
   }
   {type=MovieSceneSpawnTrack, sections [{type=MovieSceneSpawnSection, channels [None {type=MovieSceneScriptingBoolChannel, default=true}]}]}
  ]
 }
]
display_rate {denominator=1, numerator=60}
tick_resolution {denominator=1, numerator=24000}
tracks [{type=MovieSceneCameraCutTrack, sections [{type=MovieSceneCameraCutSection, end_frame=600, start_frame=0}]}]
