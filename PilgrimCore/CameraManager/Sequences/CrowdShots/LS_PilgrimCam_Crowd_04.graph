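# Level Sequence for crowd camera shot 04 in the PilgrimCore camera manager, a single ten-second cine camera shot at 60 fps.
# A camera cut track spans the shot, and the Cine Camera Actor tracks sideways along Y at fixed X and height while panning in yaw.
# Roll drifts slightly and pitch holds until both jump on the final frame.
# Component tracks hold focal length, manual focus distance and aperture constant.
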LevelSequence /PilgrimCore/CameraManager/Sequences/CrowdShots/LS_PilgrimCam_Crowd_04
bindings [
 CameraComponent {
  id=#1
  parent_id=#2
  possessed_class=CineCameraComponent
  tracks [
   {
    type=MovieSceneFloatTrack
    display_name=Current Focal Length
    sections [{type=MovieSceneFloatSection, channels [None {type=MovieSceneScriptingFloatChannel, default=25}]}]
   }
   {
    type=MovieSceneFloatTrack
    display_name="Manual Focus Distance (Focus Settings)"
    sections [{type=MovieSceneFloatSection, channels [None {type=MovieSceneScriptingFloatChannel, default=2213.873, const [2213.873,1]}]}]
   }
   {
    type=MovieSceneFloatTrack
    display_name=Current Aperture
    sections [{type=MovieSceneFloatSection, channels [None {type=MovieSceneScriptingFloatChannel, default=1}]}]
   }
  ]
 }
 Cine Camera Actor {
  id=#2
  possessed_class=CineCameraActor
  tracks [
   {
    type=MovieScene3DTransformTrack
    sections [{
      type=MovieScene3DTransformSection
      channels [
       Location.X {type=MovieSceneScriptingDoubleChannel, default=-5071.5934, const [-2145.8083,2]}
       Location.Y {type=MovieSceneScriptingDoubleChannel, default=-542.2283, keys [[0,-1219.7032],[600,-326.7801]]}
       Location.Z {type=MovieSceneScriptingDoubleChannel, default=824.5814, const [1134.5538,2]}
       Rotation.X {type=MovieSceneScriptingDoubleChannel, default=-4, keys [[0,-2],[599,-4],[600,-12.6972]]}
       Rotation.Y {type=MovieSceneScriptingDoubleChannel, default=5, keys [[0,5],[599,5],[600,-0.6066]]}
       Rotation.Z {type=MovieSceneScriptingDoubleChannel, default=-177.3, keys [[0,185.3461],[600,194.857]]}
       Scale.X {type=MovieSceneScriptingDoubleChannel, default=1, const [1,2]}
       Scale.Y {type=MovieSceneScriptingDoubleChannel, default=1, const [1,2]}
       Scale.Z {type=MovieSceneScriptingDoubleChannel, default=1, const [1,2]}
      ]
     }]
   }
   {type=MovieSceneSpawnTrack, sections [{type=MovieSceneSpawnSection, channels [None {type=MovieSceneScriptingBoolChannel, default=true}]}]}
  ]
 }
]
display_rate {denominator=1, numerator=60}
tick_resolution {denominator=1, numerator=24000}
tracks [{type=MovieSceneCameraCutTrack, sections [{type=MovieSceneCameraCutSection, end_frame=600, start_frame=0}]}]
