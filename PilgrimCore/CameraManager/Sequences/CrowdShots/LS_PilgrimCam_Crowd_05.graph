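# Level Sequence for PilgrimCam crowd shot 05, a ten-second cine camera move at 60 fps under a single camera cut.
# The Cine Camera Actor descends, mostly early in the shot, while panning, rolling and changing pitch, with a sharp pitch drop keyed on the final frame.
# Its camera component tracks widen the focal length across the shot and pull manual focus closer, with aperture held constant.
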LevelSequence /PilgrimCore/CameraManager/Sequences/CrowdShots/LS_PilgrimCam_Crowd_05
bindings [
 CameraComponent {
  id=#1
  parent_id=#2
  possessed_class=CineCameraComponent
  tracks [
   {
    type=MovieSceneFloatTrack
    display_name=Current Focal Length
    sections [{type=MovieSceneFloatSection, channels [None {type=MovieSceneScriptingFloatChannel, default=17.7437, keys [[0,24.7616],[600,17.7437]]}]}]
   }
   {
    type=MovieSceneFloatTrack
    display_name="Manual Focus Distance (Focus Settings)"
    sections [{type=MovieSceneFloatSection, channels [None {type=MovieSceneScriptingFloatChannel, default=3281.1028, keys [[0,4074.0486],[384,3281.1028]]}]}]
   }
   {
    type=MovieSceneFloatTrack
    display_name=Current Aperture
    sections [{type=MovieSceneFloatSection, channels [None {type=MovieSceneScriptingFloatChannel, default=1}]}]
   }
  ]
 }
 Cine Camera Actor {
  id=#2
  possessed_class=CineCameraActor
  tracks [
   {
    type=MovieScene3DTransformTrack
    sections [{
      type=MovieScene3DTransformSection
      channels [
       Location.X {type=MovieSceneScriptingDoubleChannel, default=-5150.5862, thinned_from=3, keys [[0,-3947.4495],[600,-3954.0532]]}
       Location.Y {type=MovieSceneScriptingDoubleChannel, default=-552.7704, keys [[0,823.2944],[211,818.8505],[600,817.7278]]}
       Location.Z {type=MovieSceneScriptingDoubleChannel, default=671.3617, keys [[0,1043.8484],[211,834.9369],[600,782.1638]]}
       Rotation.X {type=MovieSceneScriptingDoubleChannel, default=0, keys [[0,-13.8541],[211,3.4257],[600,7.7908]]}
       Rotation.Y {type=MovieSceneScriptingDoubleChannel, default=6, keys [[0,13.705],[211,6],[599,10],[600,-1.4213]]}
       Rotation.Z {type=MovieSceneScriptingDoubleChannel, default=179.1, keys [[0,-520.4055],[211,-502.0537],[600,-497.4178]]}
       Scale.X {type=MovieSceneScriptingDoubleChannel, default=1, const [1,3]}
       Scale.Y {type=MovieSceneScriptingDoubleChannel, default=1, const [1,3]}
       Scale.Z {type=MovieSceneScriptingDoubleChannel, default=1, const [1,3]}
      ]
     }]
   }
   {type=MovieSceneSpawnTrack, sections [{type=MovieSceneSpawnSection, channels [None {type=MovieSceneScriptingBoolChannel, default=true}]}]}
  ]
 }
]
display_rate {denominator=1, numerator=60}
tick_resolution {denominator=1, numerator=24000}
tracks [{type=MovieSceneCameraCutTrack, sections [{type=MovieSceneCameraCutSection, end_frame=600, start_frame=0}]}]
