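# LS_PilgrimCam_Crowd_06 is a ten-second Level Sequence for one crowd camera shot in PilgrimCore's camera manager.
# A single Cine Camera Actor makes a slow keyed drift, descending slightly with small rotation changes, under a full-length camera cut.
# Its camera component holds a wide focal length and wide-open aperture while manual focus distance pulls nearer over the shot.
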LevelSequence /PilgrimCore/CameraManager/Sequences/CrowdShots/LS_PilgrimCam_Crowd_06
bindings [
 CameraComponent {
  id=#1
  parent_id=#2
  possessed_class=CineCameraComponent
  tracks [
   {
    type=MovieSceneFloatTrack
    display_name=Current Focal Length
    sections [{type=MovieSceneFloatSection, channels [None {type=MovieSceneScriptingFloatChannel, default=14.7527}]}]
   }
   {
    type=MovieSceneFloatTrack
    display_name="Manual Focus Distance (Focus Settings)"
    sections [{type=MovieSceneFloatSection, channels [None {type=MovieSceneScriptingFloatChannel, default=1299.1249, keys [[0,1560.4856],[599,1299.1249]]}]}]
   }
   {
    type=MovieSceneFloatTrack
    display_name=Current Aperture
    sections [{type=MovieSceneFloatSection, channels [None {type=MovieSceneScriptingFloatChannel, default=1}]}]
   }
  ]
 }
 Cine Camera Actor {
  id=#2
  possessed_class=CineCameraActor
  tracks [
   {
    type=MovieScene3DTransformTrack
    sections [{
      type=MovieScene3DTransformSection
      channels [
       Location.X {type=MovieSceneScriptingDoubleChannel, default=-6326.5726, keys [[0,-5279.0005],[599,-5617.9053]]}
       Location.Y {type=MovieSceneScriptingDoubleChannel, default=-552.7704, keys [[0,4.0509],[599,-93.9561]]}
       Location.Z {type=MovieSceneScriptingDoubleChannel, default=230.4534, keys [[0,707.3372],[599,683.1024]]}
       Rotation.X {type=MovieSceneScriptingDoubleChannel, default=-2.6964, keys [[0,0.3999],[599,1.2383]]}
       Rotation.Y {type=MovieSceneScriptingDoubleChannel, default=-5.393, keys [[0,7.3491],[599,8.6717]]}
       Rotation.Z {type=MovieSceneScriptingDoubleChannel, default=42.8715, keys [[0,219.6935],[599,218.0818]]}
       Scale.X {type=MovieSceneScriptingDoubleChannel, default=1, const [1,2]}
       Scale.Y {type=MovieSceneScriptingDoubleChannel, default=1, const [1,2]}
       Scale.Z {type=MovieSceneScriptingDoubleChannel, default=1, const [1,2]}
      ]
     }]
   }
   {type=MovieSceneSpawnTrack, sections [{type=MovieSceneSpawnSection, channels [None {type=MovieSceneScriptingBoolChannel, default=true}]}]}
  ]
 }
]
display_rate {denominator=1, numerator=60}
tick_resolution {denominator=1, numerator=24000}
tracks [{type=MovieSceneCameraCutTrack, sections [{type=MovieSceneCameraCutSection, end_frame=600, start_frame=0}]}]
