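# Camera sequence for one crowd shot (number 07) in the PilgrimCore camera manager.
# A Cine Camera Actor with a spawn track carries its camera component, which holds a fixed transform, focal length, manual focus distance and wide-open aperture.
# A camera shake track on the component supplies the only motion in the otherwise static framing.
# A single camera cut section spans the 600-frame shot.
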LevelSequence /PilgrimCore/CameraManager/Sequences/CrowdShots/LS_PilgrimCam_Crowd_07
bindings [
 CameraComponent {
  id=#1
  parent_id=#2
  possessed_class=CineCameraComponent
  tracks [
   {
    type=MovieSceneFloatTrack
    display_name=Current Focal Length
    sections [{type=MovieSceneFloatSection, channels [None {type=MovieSceneScriptingFloatChannel, default=38.3328}]}]
   }
   {
    type=MovieSceneFloatTrack
    display_name="Manual Focus Distance (Focus Settings)"
    sections [{type=MovieSceneFloatSection, channels [None {type=MovieSceneScriptingFloatChannel, default=4492.417}]}]
   }
   {
    type=MovieSceneFloatTrack
    display_name=Current Aperture
    sections [{type=MovieSceneFloatSection, channels [None {type=MovieSceneScriptingFloatChannel, default=1}]}]
   }
   {
    type=MovieScene3DTransformTrack
    sections [{
      type=MovieScene3DTransformSection
      channels [
       Location.X {type=MovieSceneScriptingDoubleChannel, default=63.0406, const [0,1]}
       Location.Y {type=MovieSceneScriptingDoubleChannel, default=168.397, const [0,1]}
       Location.Z {type=MovieSceneScriptingDoubleChannel, default=-23.2676, const [0,1]}
       Rotation.X {type=MovieSceneScriptingDoubleChannel, default=5.6681, const [0,1]}
       Rotation.Y {type=MovieSceneScriptingDoubleChannel, default=-9.5215, const [0,1]}
       Rotation.Z {type=MovieSceneScriptingDoubleChannel, default=38.8474, const [0,1]}
       Scale.X {type=MovieSceneScriptingDoubleChannel, default=1, const [1,1]}
       Scale.Y {type=MovieSceneScriptingDoubleChannel, default=1, const [1,1]}
       Scale.Z {type=MovieSceneScriptingDoubleChannel, default=1, const [1,1]}
      ]
     }]
   }
   {type=MovieSceneCameraShakeTrack, sections [{type=MovieSceneCameraShakeSection, end_frame=618}]}
  ]
 }
 Cine Camera Actor {
  id=#2
  possessed_class=CineCameraActor
  tracks [{type=MovieSceneSpawnTrack, sections [{type=MovieSceneSpawnSection, channels [None {type=MovieSceneScriptingBoolChannel, default=true}]}]}]
 }
]
display_rate {denominator=1, numerator=60}
tick_resolution {denominator=1, numerator=24000}
tracks [{type=MovieSceneCameraCutTrack, sections [{type=MovieSceneCameraCutSection, end_frame=600, start_frame=0}]}]
